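# Ambient sound cue for a destroyed Rift Gate device, layering looping fire, electrical hum and random sparks.
# The root mixer blends a distant-attenuated electrical loop, a louder close-attenuated fire loop and a distant-attenuated spark layer.
# Two spark chains each loop a modulated random pick from fourteen waves, with the random delay outside the loop, staggering only their starts.
# A quality switch feeds sparks to three slots, leaving the fourth empty.
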SoundCue /Game/Athena/Sounds/Props/RiftGate/Rift_Gate_Device_Destroyed_Amb_Loop_Cue
root=Mixer_0
nodes [
 {id=Attenuation_0, children [WavePlayer_0], properties {attenuationSettings=&/Rift_Gate_Close_Att}}
 {id=Attenuation_1, children [WavePlayer_1], properties {attenuationSettings=&/Rift_Gate_Distant_Att}}
 {id=Attenuation_2, children [Mixer_1], properties {attenuationSettings=&Rift_Gate_Distant_Att}}
 {id=Delay_0, children [Looping_0], properties {delayMax=3, delayMin=1}}
 {id=Delay_1, children [Looping_1], properties {delayMax=1, delayMin=0.5}}
 {id=Looping_0, children [Modulator_0]}
 {id=Looping_1, children [Modulator_1]}
 {id=Mixer_0, children [Attenuation_1, Attenuation_0, QualityLevel_0], properties {inputVolume [0.7,2,0.7]}}
 {id=Mixer_1, children [Delay_0,Delay_1], properties {inputVolume [1,1]}}
 {id=Modulator_0, children [Random_0]}
 {id=Modulator_1, children [Random_1]}
 {id=QualityLevel_0, children [Attenuation_2, Attenuation_2, Attenuation_2, null]}
 {
  id=Random_0
  children [
   WavePlayer_2
   WavePlayer_3
   WavePlayer_4
   WavePlayer_5
   WavePlayer_6
   WavePlayer_7
   WavePlayer_8
   WavePlayer_9
   WavePlayer_10
   WavePlayer_11
   WavePlayer_12
   WavePlayer_13
   WavePlayer_14
   WavePlayer_15
  ]
  properties {bShouldExcludeFromBranchCulling=true, weights [1,1,1,1,1,1,1,1,1,1,1,1,1,1]}
 }
 {
  id=Random_1
  children [
   WavePlayer_16
   WavePlayer_17
   WavePlayer_18
   WavePlayer_19
   WavePlayer_20
   WavePlayer_21
   WavePlayer_22
   WavePlayer_23
   WavePlayer_24
   WavePlayer_25
   WavePlayer_26
   WavePlayer_27
   WavePlayer_28
   WavePlayer_29
  ]
  properties {bShouldExcludeFromBranchCulling=true, weights [1,1,1,1,1,1,1,1,1,1,1,1,1,1]}
 }
 {id=WavePlayer_0, properties {bLooping=true, soundWaveAssetPtr=&/Rift_Gate_Device_Amb_Destroyed_Fire_Loop}}
 {id=WavePlayer_1, properties {bLooping=true, soundWaveAssetPtr=&/Rift_Gate_Device_Amb_Destroyed_Elec_Loop}}
 {id=WavePlayer_10, properties {soundWaveAssetPtr=&/Rift_Gate_Device_Amb_Elec_Spark_09}}
 {id=WavePlayer_11, properties {soundWaveAssetPtr=&/Rift_Gate_Device_Amb_Elec_Spark_10}}
 {id=WavePlayer_12, properties {soundWaveAssetPtr=&/Rift_Gate_Device_Amb_Elec_Spark_11}}
 {id=WavePlayer_13, properties {soundWaveAssetPtr=&/Rift_Gate_Device_Amb_Elec_Spark_12}}
 {id=WavePlayer_14, properties {soundWaveAssetPtr=&/Rift_Gate_Device_Amb_Elec_Spark_13}}
 {id=WavePlayer_15, properties {soundWaveAssetPtr=&/Rift_Gate_Device_Amb_Elec_Spark_14}}
 {id=WavePlayer_16, properties {soundWaveAssetPtr=&/Rift_Gate_Device_Amb_Elec_Spark_01}}
 {id=WavePlayer_17, properties {soundWaveAssetPtr=&/Rift_Gate_Device_Amb_Elec_Spark_02}}
 {id=WavePlayer_18, properties {soundWaveAssetPtr=&/Rift_Gate_Device_Amb_Elec_Spark_03}}
 {id=WavePlayer_19, properties {soundWaveAssetPtr=&/Rift_Gate_Device_Amb_Elec_Spark_04}}
 {id=WavePlayer_2, properties {soundWaveAssetPtr=&Rift_Gate_Device_Amb_Elec_Spark_01}}
 {id=WavePlayer_20, properties {soundWaveAssetPtr=&/Rift_Gate_Device_Amb_Elec_Spark_05}}
 {id=WavePlayer_21, properties {soundWaveAssetPtr=&/Rift_Gate_Device_Amb_Elec_Spark_06}}
 {id=WavePlayer_22, properties {soundWaveAssetPtr=&/Rift_Gate_Device_Amb_Elec_Spark_07}}
 {id=WavePlayer_23, properties {soundWaveAssetPtr=&/Rift_Gate_Device_Amb_Elec_Spark_08}}
 {id=WavePlayer_24, properties {soundWaveAssetPtr=&Rift_Gate_Device_Amb_Elec_Spark_09}}
 {id=WavePlayer_25, properties {soundWaveAssetPtr=&Rift_Gate_Device_Amb_Elec_Spark_10}}
 {id=WavePlayer_26, properties {soundWaveAssetPtr=&Rift_Gate_Device_Amb_Elec_Spark_11}}
 {id=WavePlayer_27, properties {soundWaveAssetPtr=&Rift_Gate_Device_Amb_Elec_Spark_12}}
 {id=WavePlayer_28, properties {soundWaveAssetPtr=&Rift_Gate_Device_Amb_Elec_Spark_13}}
 {id=WavePlayer_29, properties {soundWaveAssetPtr=&Rift_Gate_Device_Amb_Elec_Spark_14}}
 {id=WavePlayer_3, properties {soundWaveAssetPtr=&Rift_Gate_Device_Amb_Elec_Spark_02}}
 {id=WavePlayer_4, properties {soundWaveAssetPtr=&Rift_Gate_Device_Amb_Elec_Spark_03}}
 {id=WavePlayer_5, properties {soundWaveAssetPtr=&Rift_Gate_Device_Amb_Elec_Spark_04}}
 {id=WavePlayer_6, properties {soundWaveAssetPtr=&Rift_Gate_Device_Amb_Elec_Spark_05}}
 {id=WavePlayer_7, properties {soundWaveAssetPtr=&Rift_Gate_Device_Amb_Elec_Spark_06}}
 {id=WavePlayer_8, properties {soundWaveAssetPtr=&Rift_Gate_Device_Amb_Elec_Spark_07}}
 {id=WavePlayer_9, properties {soundWaveAssetPtr=&Rift_Gate_Device_Amb_Elec_Spark_08}}
]
properties {maxDistance=3500, soundClassObject=&Fort_Ambient_Props_Athena, volumeMultiplier=0.752}
