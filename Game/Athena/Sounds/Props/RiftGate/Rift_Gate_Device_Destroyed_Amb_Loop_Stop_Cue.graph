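# Sound cue for the destroyed Rift Gate device, which by its name plays when the ambient loop stops, routed to the Athena ambient props sound class.
# A Mixer layers the destroyed sound with a quieter electrical spark, each through distant attenuation.
# The spark is picked at random from fourteen variants, then modulated and briefly delayed.
# A quality switch shares that spark branch across its first three slots and leaves the last empty.
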SoundCue /Game/Athena/Sounds/Props/RiftGate/Rift_Gate_Device_Destroyed_Amb_Loop_Stop_Cue
root=Mixer_0
nodes [
 {id=Attenuation_0, children [WavePlayer_8], properties {attenuationSettings=&/Rift_Gate_Distant_Att}}
 {id=Attenuation_2, children [Delay_2], properties {attenuationSettings=&Rift_Gate_Distant_Att}}
 {id=Delay_2, children [Modulator_0], properties {delayMax=0.5, delayMin=0.5}}
 {id=Mixer_0, children [Attenuation_0, QualityLevel_0], properties {inputVolume [1,0.5]}}
 {id=Modulator_0, children [Random_0]}
 {id=QualityLevel_0, children [Attenuation_2, Attenuation_2, Attenuation_2, null]}
 {
  id=Random_0
  children [
   WavePlayer_2
   WavePlayer_3
   WavePlayer_4
   WavePlayer_5
   WavePlayer_6
   WavePlayer_7
   WavePlayer_0
   WavePlayer_9
   WavePlayer_10
   WavePlayer_11
   WavePlayer_12
   WavePlayer_13
   WavePlayer_14
   WavePlayer_15
  ]
  properties {bShouldExcludeFromBranchCulling=true, weights [1,1,1,1,1,1,1,1,1,1,1,1,1,1]}
 }
 {id=WavePlayer_0, properties {soundWaveAssetPtr=&/Rift_Gate_Device_Amb_Elec_Spark_07}}
 {id=WavePlayer_10, properties {soundWaveAssetPtr=&/Rift_Gate_Device_Amb_Elec_Spark_09}}
 {id=WavePlayer_11, properties {soundWaveAssetPtr=&/Rift_Gate_Device_Amb_Elec_Spark_10}}
 {id=WavePlayer_12, properties {soundWaveAssetPtr=&/Rift_Gate_Device_Amb_Elec_Spark_11}}
 {id=WavePlayer_13, properties {soundWaveAssetPtr=&/Rift_Gate_Device_Amb_Elec_Spark_12}}
 {id=WavePlayer_14, properties {soundWaveAssetPtr=&/Rift_Gate_Device_Amb_Elec_Spark_13}}
 {id=WavePlayer_15, properties {soundWaveAssetPtr=&/Rift_Gate_Device_Amb_Elec_Spark_14}}
 {id=WavePlayer_2, properties {soundWaveAssetPtr=&/Rift_Gate_Device_Amb_Elec_Spark_01}}
 {id=WavePlayer_3, properties {soundWaveAssetPtr=&/Rift_Gate_Device_Amb_Elec_Spark_02}}
 {id=WavePlayer_4, properties {soundWaveAssetPtr=&/Rift_Gate_Device_Amb_Elec_Spark_03}}
 {id=WavePlayer_5, properties {soundWaveAssetPtr=&/Rift_Gate_Device_Amb_Elec_Spark_04}}
 {id=WavePlayer_6, properties {soundWaveAssetPtr=&/Rift_Gate_Device_Amb_Elec_Spark_05}}
 {id=WavePlayer_7, properties {soundWaveAssetPtr=&/Rift_Gate_Device_Amb_Elec_Spark_06}}
 {id=WavePlayer_8, properties {soundWaveAssetPtr=&/Rift_Gate_Device_Destroyed}}
 {id=WavePlayer_9, properties {soundWaveAssetPtr=&/Rift_Gate_Device_Amb_Elec_Spark_08}}
]
properties {maxDistance=3500, soundClassObject=&Fort_Ambient_Props_Athena, volumeMultiplier=0.877}
