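# Ambient Sound Cue for the powered Rift Gate device, mixing close and distant electrical loops, a tonal loop, looping random whooshes and two randomly delayed spark layers.
# A CalendarControl parameter continuously modulates each layer's pitch and volume, with higher input thresholds for tonal, whoosh and spark volume.
# Close and distant attenuation assets shape the layers by distance.
# A quality-level node gates the sparks, leaving its last slot empty.
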SoundCue /Game/Athena/Sounds/Props/RiftGate/Rift_Gate_Device_Powered_Amb_Loop_Cue
root=Mixer_0
nodes [
 {id=Attenuation_0, children [WavePlayer_0], properties {attenuationSettings=&/Rift_Gate_Close_Att}}
 {id=Attenuation_1, children [WavePlayer_1], properties {attenuationSettings=&/Rift_Gate_Distant_Att}}
 {id=Attenuation_2, children [Looping_0], properties {attenuationSettings=&Rift_Gate_Distant_Att}}
 {id=Attenuation_3, children [Mixer_1], properties {attenuationSettings=&Rift_Gate_Distant_Att}}
 {id=Attenuation_4, children [WavePlayer_36], properties {attenuationSettings=&Rift_Gate_Close_Att}}
 {id=Delay_0, children [Modulator_0]}
 {id=Delay_1, children [Looping_1], properties {delayMax=2, delayMin=0.5}}
 {id=Delay_2, children [Looping_2], properties {delayMax=5, delayMin=1}}
 {id=Looping_0, children [Delay_0]}
 {id=Looping_1, children [Modulator_1]}
 {id=Looping_2, children [Modulator_2]}
 {
  id=Mixer_0
  children [ModulatorContinuous_0, ModulatorContinuous_1, ModulatorContinuous_2, ModulatorContinuous_4, ModulatorContinuous_3]
  properties {inputVolume [2,2.5,4,2.5,1]}
 }
 {id=Mixer_1, children [Delay_1,Delay_2], properties {inputVolume [1,1]}}
 {
  id=ModulatorContinuous_0
  children [Attenuation_1]
  properties {
   pitchModulationParams {minOutput=0.75, parameterName=CalendarControl}
   volumeModulationParams {default=0, minOutput=1, parameterName=CalendarControl}
  }
 }
 {
  id=ModulatorContinuous_1
  children [Attenuation_0]
  properties {
   pitchModulationParams {minOutput=0.75, parameterName=CalendarControl}
   volumeModulationParams {default=0, minOutput=1, parameterName=CalendarControl}
  }
 }
 {
  id=ModulatorContinuous_2
  children [Attenuation_4]
  properties {
   pitchModulationParams {minOutput=0.85, parameterName=CalendarControl}
   volumeModulationParams {default=0, minInput=0.75, paramMode=MPM_Abs, parameterName=CalendarControl}
  }
 }
 {
  id=ModulatorContinuous_3
  children [QualityLevel_0]
  properties {
   pitchModulationParams {parameterName=CalendarControl}
   volumeModulationParams {default=0, minInput=0.5, paramMode=MPM_Abs, parameterName=CalendarControl}
  }
 }
 {
  id=ModulatorContinuous_4
  children [Attenuation_2]
  properties {
   pitchModulationParams {parameterName=CalendarControl}
   volumeModulationParams {default=0, minInput=0.5, paramMode=MPM_Abs, parameterName=CalendarControl}
  }
 }
 {id=Modulator_0, children [Random_0]}
 {id=Modulator_1, children [Random_1], properties {pitchMax=1.2, volumeMin=0.75}}
 {id=Modulator_2, children [Random_2], properties {pitchMax=1.25, volumeMin=0.5}}
 {id=QualityLevel_0, children [Attenuation_3, Attenuation_3, Attenuation_3, null]}
 {
  id=Random_0
  children [WavePlayer_2,WavePlayer_3,WavePlayer_4,WavePlayer_5,WavePlayer_6,WavePlayer_7]
  properties {bShouldExcludeFromBranchCulling=true, weights [1,1,1,1,1,1]}
 }
 {
  id=Random_1
  children [
   WavePlayer_8
   WavePlayer_9
   WavePlayer_10
   WavePlayer_11
   WavePlayer_12
   WavePlayer_13
   WavePlayer_14
   WavePlayer_15
   WavePlayer_16
   WavePlayer_17
   WavePlayer_18
   WavePlayer_19
   WavePlayer_20
   WavePlayer_21
  ]
  properties {bShouldExcludeFromBranchCulling=true, weights [1,1,1,1,1,1,1,1,1,1,1,1,1,1]}
 }
 {
  id=Random_2
  children [
   WavePlayer_22
   WavePlayer_23
   WavePlayer_24
   WavePlayer_25
   WavePlayer_26
   WavePlayer_27
   WavePlayer_28
   WavePlayer_29
   WavePlayer_30
   WavePlayer_31
   WavePlayer_32
   WavePlayer_33
   WavePlayer_34
   WavePlayer_35
  ]
  properties {bShouldExcludeFromBranchCulling=true, weights [1,1,1,1,1,1,1,1,1,1,1,1,1,1]}
 }
 {id=WavePlayer_0, properties {bLooping=true, soundWaveAssetPtr=&/Rift_Gate_Device_Amb_Loop_Close_Elec_01}}
 {id=WavePlayer_1, properties {bLooping=true, soundWaveAssetPtr=&/Rift_Gate_Device_Amb_Loop_Distant_Elec_01}}
 {id=WavePlayer_10, properties {soundWaveAssetPtr=&/Rift_Gate_Device_Amb_Elec_Spark_03}}
 {id=WavePlayer_11, properties {soundWaveAssetPtr=&/Rift_Gate_Device_Amb_Elec_Spark_04}}
 {id=WavePlayer_12, properties {soundWaveAssetPtr=&/Rift_Gate_Device_Amb_Elec_Spark_05}}
 {id=WavePlayer_13, properties {soundWaveAssetPtr=&/Rift_Gate_Device_Amb_Elec_Spark_06}}
 {id=WavePlayer_14, properties {soundWaveAssetPtr=&/Rift_Gate_Device_Amb_Elec_Spark_07}}
 {id=WavePlayer_15, properties {soundWaveAssetPtr=&/Rift_Gate_Device_Amb_Elec_Spark_08}}
 {id=WavePlayer_16, properties {soundWaveAssetPtr=&/Rift_Gate_Device_Amb_Elec_Spark_09}}
 {id=WavePlayer_17, properties {soundWaveAssetPtr=&/Rift_Gate_Device_Amb_Elec_Spark_10}}
 {id=WavePlayer_18, properties {soundWaveAssetPtr=&/Rift_Gate_Device_Amb_Elec_Spark_11}}
 {id=WavePlayer_19, properties {soundWaveAssetPtr=&/Rift_Gate_Device_Amb_Elec_Spark_12}}
 {id=WavePlayer_2, properties {soundWaveAssetPtr=&/Rift_Gate_Device_Amb_Close_Whoosh_01}}
 {id=WavePlayer_20, properties {soundWaveAssetPtr=&/Rift_Gate_Device_Amb_Elec_Spark_13}}
 {id=WavePlayer_21, properties {soundWaveAssetPtr=&/Rift_Gate_Device_Amb_Elec_Spark_14}}
 {id=WavePlayer_22, properties {soundWaveAssetPtr=&/Rift_Gate_Device_Amb_Elec_Spark_01}}
 {id=WavePlayer_23, properties {soundWaveAssetPtr=&/Rift_Gate_Device_Amb_Elec_Spark_02}}
 {id=WavePlayer_24, properties {soundWaveAssetPtr=&Rift_Gate_Device_Amb_Elec_Spark_03}}
 {id=WavePlayer_25, properties {soundWaveAssetPtr=&Rift_Gate_Device_Amb_Elec_Spark_04}}
 {id=WavePlayer_26, properties {soundWaveAssetPtr=&Rift_Gate_Device_Amb_Elec_Spark_05}}
 {id=WavePlayer_27, properties {soundWaveAssetPtr=&Rift_Gate_Device_Amb_Elec_Spark_06}}
 {id=WavePlayer_28, properties {soundWaveAssetPtr=&Rift_Gate_Device_Amb_Elec_Spark_07}}
 {id=WavePlayer_29, properties {soundWaveAssetPtr=&Rift_Gate_Device_Amb_Elec_Spark_08}}
 {id=WavePlayer_3, properties {soundWaveAssetPtr=&/Rift_Gate_Device_Amb_Close_Whoosh_02}}
 {id=WavePlayer_30, properties {soundWaveAssetPtr=&Rift_Gate_Device_Amb_Elec_Spark_09}}
 {id=WavePlayer_31, properties {soundWaveAssetPtr=&Rift_Gate_Device_Amb_Elec_Spark_10}}
 {id=WavePlayer_32, properties {soundWaveAssetPtr=&Rift_Gate_Device_Amb_Elec_Spark_11}}
 {id=WavePlayer_33, properties {soundWaveAssetPtr=&Rift_Gate_Device_Amb_Elec_Spark_12}}
 {id=WavePlayer_34, properties {soundWaveAssetPtr=&Rift_Gate_Device_Amb_Elec_Spark_13}}
 {id=WavePlayer_35, properties {soundWaveAssetPtr=&Rift_Gate_Device_Amb_Elec_Spark_14}}
 {
  id=WavePlayer_36
  properties {bLooping=true, soundWaveAssetPtr=&/Rift_Gate_Device_Powered_Tonal_Amb_Loop_01}
 }
 {id=WavePlayer_4, properties {soundWaveAssetPtr=&/Rift_Gate_Device_Amb_Close_Whoosh_03}}
 {id=WavePlayer_5, properties {soundWaveAssetPtr=&/Rift_Gate_Device_Amb_Close_Whoosh_04}}
 {id=WavePlayer_6, properties {soundWaveAssetPtr=&/Rift_Gate_Device_Amb_Close_Whoosh_05}}
 {id=WavePlayer_7, properties {soundWaveAssetPtr=&/Rift_Gate_Device_Amb_Close_Whoosh_06}}
 {id=WavePlayer_8, properties {soundWaveAssetPtr=&Rift_Gate_Device_Amb_Elec_Spark_01}}
 {id=WavePlayer_9, properties {soundWaveAssetPtr=&Rift_Gate_Device_Amb_Elec_Spark_02}}
]
properties {maxDistance=3500, soundClassObject=&Fort_Ambient_Props_Athena, volumeMultiplier=1.002}
