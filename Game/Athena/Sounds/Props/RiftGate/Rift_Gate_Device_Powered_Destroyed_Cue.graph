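# Sound cue for the Athena Rift Gate prop, by its name the sound of the powered device being destroyed.
# A single wave player feeds an attenuation node that uses the shared distant Rift Gate attenuation asset.
# The cue plays in the Athena ambient props sound class at slightly reduced volume, with no randomization or mixing.
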SoundCue /Game/Athena/Sounds/Props/RiftGate/Rift_Gate_Device_Powered_Destroyed_Cue
root=Attenuation_2
nodes [
 {id=Attenuation_2, children [WavePlayer_8], properties {attenuationSettings=&/Rift_Gate_Distant_Att}}
 {id=WavePlayer_8, properties {soundWaveAssetPtr=&/Rift_Gate_Device_Powered_Destroyed}}
]
properties {maxDistance=3500, soundClassObject=&Fort_Ambient_Props_Athena, volumeMultiplier=0.877}
