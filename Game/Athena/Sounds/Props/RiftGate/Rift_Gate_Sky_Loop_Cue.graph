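# Ambient sound cue for the Rift Gate sky in Athena, layering two looping beds with randomized rumbles after a short fixed delay.
# The CalendarControl parameter modulates each layer's volume over different ranges and the loops' pitch, so intensity builds as it changes.
# Both beds fade in through envelopers.
# Two looping random rumble chains with different delays mix behind a quality switch that omits them at one level.
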SoundCue /Game/Athena/Sounds/Props/RiftGate/Rift_Gate_Sky_Loop_Cue
root=Delay_1
nodes [
 {id=Delay_0, children [Looping_0], properties {delayMax=2, delayMin=0.5}}
 {id=Delay_1, children [Mixer_0], properties {delayMax=1.5, delayMin=1.5}}
 {id=Delay_2, children [Looping_1], properties {delayMax=5, delayMin=2}}
 {
  id=Enveloper_1
  children [WavePlayer_0]
  properties {pitchCurve {editorCurveData {keys [{time=0}]}}, volumeCurve {editorCurveData {keys [<at 0 0 RCIM_Cubic>,<at 3 1>]}}}
 }
 {
  id=Enveloper_2
  children [WavePlayer_1]
  properties {pitchCurve {editorCurveData {keys [{time=0}]}}, volumeCurve {editorCurveData {keys [<at 0 0 RCIM_Cubic>,<at 3 1>]}}}
 }
 {id=Looping_0, children [Random_1]}
 {id=Looping_1, children [Random_2]}
 {id=Mixer_0, children [ModulatorContinuous_2, ModulatorContinuous_1, QualityLevel_0], properties {inputVolume [1,1,1]}}
 {id=Mixer_1, children [Delay_0,Delay_2], properties {inputVolume [1,2.5]}}
 {
  id=ModulatorContinuous_1
  children [Enveloper_2]
  properties {
   pitchModulationParams {minOutput=0.75, parameterName=CalendarControl}
   volumeModulationParams {default=0, minOutput=0.3, parameterName=CalendarControl}
  }
 }
 {
  id=ModulatorContinuous_2
  children [Enveloper_1]
  properties {
   pitchModulationParams {minOutput=0.75, parameterName=CalendarControl}
   volumeModulationParams {default=0, minInput=0.4, parameterName=CalendarControl}
  }
 }
 {
  id=ModulatorContinuous_3
  children [Mixer_1]
  properties {
   pitchModulationParams {minOutput=1, parameterName=CalendarControl}
   volumeModulationParams {default=0, minInput=0.75, parameterName=CalendarControl}
  }
 }
 {id=QualityLevel_0, children [ModulatorContinuous_3, ModulatorContinuous_3, ModulatorContinuous_3, null]}
 {
  id=Random_1
  children [WavePlayer_2, WavePlayer_3, WavePlayer_4, WavePlayer_5, WavePlayer_6, WavePlayer_7, WavePlayer_8, WavePlayer_9, WavePlayer_10]
  properties {bShouldExcludeFromBranchCulling=true, weights [1,1,1,1,1,1,1,1,1]}
 }
 {
  id=Random_2
  children [WavePlayer_11, WavePlayer_12, WavePlayer_13, WavePlayer_14, WavePlayer_15, WavePlayer_16, WavePlayer_17, WavePlayer_18, WavePlayer_19]
  properties {bShouldExcludeFromBranchCulling=true, weights [1,1,1,1,1,1,1,1,1]}
 }
 {id=WavePlayer_0, properties {bLooping=true, soundWaveAssetPtr=&/Rift_Gate_Sky_Loop_B}}
 {id=WavePlayer_1, properties {bLooping=true, soundWaveAssetPtr=&/Rift_Gate_Sky_Loop_A}}
 {id=WavePlayer_10, properties {soundWaveAssetPtr=&/Rift_Gate_Sky_Rumble_09}}
 {id=WavePlayer_11, properties {soundWaveAssetPtr=&/Rift_Gate_Sky_Rumble_01}}
 {id=WavePlayer_12, properties {soundWaveAssetPtr=&/Rift_Gate_Sky_Rumble_02}}
 {id=WavePlayer_13, properties {soundWaveAssetPtr=&/Rift_Gate_Sky_Rumble_03}}
 {id=WavePlayer_14, properties {soundWaveAssetPtr=&/Rift_Gate_Sky_Rumble_04}}
 {id=WavePlayer_15, properties {soundWaveAssetPtr=&/Rift_Gate_Sky_Rumble_05}}
 {id=WavePlayer_16, properties {soundWaveAssetPtr=&/Rift_Gate_Sky_Rumble_06}}
 {id=WavePlayer_17, properties {soundWaveAssetPtr=&/Rift_Gate_Sky_Rumble_07}}
 {id=WavePlayer_18, properties {soundWaveAssetPtr=&/Rift_Gate_Sky_Rumble_08}}
 {id=WavePlayer_19, properties {soundWaveAssetPtr=&Rift_Gate_Sky_Rumble_09}}
 {id=WavePlayer_2, properties {soundWaveAssetPtr=&Rift_Gate_Sky_Rumble_01}}
 {id=WavePlayer_3, properties {soundWaveAssetPtr=&Rift_Gate_Sky_Rumble_02}}
 {id=WavePlayer_4, properties {soundWaveAssetPtr=&Rift_Gate_Sky_Rumble_03}}
 {id=WavePlayer_5, properties {soundWaveAssetPtr=&Rift_Gate_Sky_Rumble_04}}
 {id=WavePlayer_6, properties {soundWaveAssetPtr=&Rift_Gate_Sky_Rumble_05}}
 {id=WavePlayer_7, properties {soundWaveAssetPtr=&Rift_Gate_Sky_Rumble_06}}
 {id=WavePlayer_8, properties {soundWaveAssetPtr=&Rift_Gate_Sky_Rumble_07}}
 {id=WavePlayer_9, properties {soundWaveAssetPtr=&Rift_Gate_Sky_Rumble_08}}
]
properties {
 attenuationSettings=&/Rift_Gate_Sky_Loop_Att
 bOverrideConcurrency=true
 maxDistance=2097152
 soundClassObject=&Fort_Ambient_Athena
 volumeMultiplier=0.376
 concurrencyOverrides {maxCount=5, resolutionRule=StopOldest}
}
